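# Niagara system for the prop-spawn effect of the SlugRipple player emote, apparently built from lightweight (LWE) emitters.
# It has two sets of airglow, smoke and sparkle emitters, with no renderer details listed.
# Two system-level variable assignments, alongside vector lerp, quaternion lerp and float-curve inputs, suggest computed position and rotation.
# A User.NormalizedNotifyProgress float, starting at zero, exposes animation notify progress, likely to drive those interpolations.
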NiagaraSystem /BRCosmetics/Animation/Game/MainPlayer/Emotes/SlugRipple/FX/NS_SlugRipple_PropSpawn_LWE
dependencies {
 usedDataInterfaces [&NiagaraDataInterfaceCurve]
 usedDynamicInputs [
  &Lerp_Vector
  &FloatFromCurve
  &LerpQuaternion
 ]
 usedModules [
  &SystemState
  &self:SystemSpawnScript.SystemScriptSource.SystemScriptGraph.NiagaraNodeAssignment_0.SetVariables_#1
  &self:SystemSpawnScript.SystemScriptSource.SystemScriptGraph.NiagaraNodeAssignment_1.SetVariables_#2
 ]
}
emitters {
 AirglowLWE {rendererClasses [], renderers []}
 AirglowLWE001 {rendererClasses [], renderers []}
 SmokeLWE {rendererClasses [], renderers []}
 SmokeLWE001 {rendererClasses [], renderers []}
 SparklesLWE {rendererClasses [], renderers []}
 SparklesLWE001 {rendererClasses [], renderers []}
}
summary {
 systemName=NS_SlugRipple_PropSpawn_LWE
 emitters [
  {emitterName=SparklesLWE, rendererClasses []}
  {emitterName=AirglowLWE, rendererClasses []}
  {emitterName=SmokeLWE, rendererClasses []}
  {emitterName=SparklesLWE001, rendererClasses []}
  {emitterName=AirglowLWE001, rendererClasses []}
  {emitterName=SmokeLWE001, rendererClasses []}
 ]
 userVariables [User.NormalizedNotifyProgress {type=NiagaraFloat, defaultValue=0, description=""}]
}
